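# Niagara system for the LightTrail vehicle boost cosmetic, drawing ribbon trails from the car's boost ports.
# One parent emitter spawns a ribbon per filtered socket, positioned with skeletal and car offsets, feeding four ribbon renderers.
# Painted-variant overrides color the base and three extra layers, with most defaults switched by game mode.
# Particles move backward, inherit fading source movement outside frontend preview, and pulse brightness and alpha over life.
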NiagaraSystem /VehicleCosmetics/Boosts/BOOST_LightTrail/FX/NS_Boost_LightTrail
dependencies {
 usedDataInterfaces [&NiagaraDataInterfaceCurve, &NiagaraDataInterfaceVectorCurve, &NiagaraDataInterfaceSkeletalMesh]
 usedDynamicInputs []
 usedModules [
  &self:SystemSpawnScript.SystemScriptSource.SystemScriptGraph.NiagaraNodeAssignment_0.SetVariables_#1
  &SystemState
 ]
}
emitters {NE_VC_Ribbon_Parent {
  emitterUpdateScript [
   &EmitterState(Inactive Response=::NewEnumerator1, Life Cycle Mode=::NewEnumerator1, Loop Duration=1, UseLoopDelay=false)
   &VC_FilteredSocketCount()
   &VC_RateQualityHandler()
   &VC_SpawnRibbon_v1(Number of Ribbons=$Emitter.PortCount, SpawnRate=60)
  ]
  particleSpawnScript [
   &VC_PaintedVariantsOverride(
    Black=<rgba 0.03 0.03 0.03 0.2>
    BurntSienna=<rgba 0.1458 0.045 0.018 0.1>
    Crimson=<rgba 1 0 0.0271 0.03>
    DefaultColor=&MergeHelper_ColorSwitch(
     0=<rgba 0.049 0.0638 0.4948 0.4>
     1=<rgba 0.049 0.0638 0.4948 0.4>
     2=<rgba 0.049 0.0638 0.4948 0.4>
     3=?
     Index=$System.GameModeID
     Num=::NewEnumerator1
    )
    ForestGreen=<rgba 0 0.5 0.0253 0.02>
    Grey=<rgba 0.15 0.15 0.15 0.1>
    Lime=<rgba 0.2943 0.5885 0 0.025>
    Orange=<rgba 0.6724 0.2346 0 0.07>
    Pink=<rgba 0.2865 0.0343 0.1948 0.1>
    Purple=<rgba 0.2789 0 0.6795 0.15>
    R1 Black:LinearColor=<rgba 0.09 0.09 0.09 14>
    R1 BurntSienna:LinearColor=<rgba 0.15 0.0797 0.039 13>
    R1 Cobalt:LinearColor=<rgba 0 0.0185 1 1>
    R1 Crimson:LinearColor=<rgba 0.5 0.01 0.0233 13>
    R1 DefaultColor=&MergeHelper_ColorSwitch(0=<rgba 0.5833 0.6463 3.5 2.5>, 1=<rgba 0.6667 1.5 4 2.5>, 2=<rgba 0.6667 1.5 4 2.5>, 3=?, Index=$System.GameModeID, Num=::NewEnumerator1)
    R1 ForestGreen:LinearColor=<rgba 0.066 0.6 0.1105 13>
    R1 Grey:LinearColor=<rgba 0.32 0.32 0.32 13>
    R1 Lime:LinearColor=<rgba 0.3843 0.58 0.0696 13>
    R1 Orange:LinearColor=<rgba 0.61 0.3562 0.1342 15>
    R1 Pink:LinearColor=<rgba 0.51 0.1581 0.3927 12>
    R1 Purple:LinearColor=<rgba 0.3819 0.057 0.57 14>
    R1 Saffron:LinearColor=<rgba 0.8 0.6667 0 12>
    R1 SkyBlue:LinearColor=<rgba 0.0825 0.4409 0.55 13>
    R1 TitaniumWhite:LinearColor=<rgba 1 0.9411 0.7917 4.5>
    R2 Black:LinearColor=<rgba 0.02 0.02 0.02 15>
    R2 BurntSienna:LinearColor=<rgba 1 0.3031 0.2604 1.5>
    R2 Crimson:LinearColor=<rgba 6 0.4375 0.5883 1>
    R2 DefaultColor:LinearColor=&MergeHelper_ColorSwitch(0=<rgba 1 1.0333 2 5>, 1=<rgba 3.5 3.6167 7 5>, 2=<rgba 3.5 3.6167 7 5>, 3=?, Index=$System.GameModeID, Num=::NewEnumerator1)
    R2 ForestGreen:LinearColor=<rgba 1.8827 6 1.75 1>
    R2 Grey:LinearColor=<rgba 0.8385 1 0.8679 3>
    R2 Lime:LinearColor=<rgba 0.8333 1 0 8>
    R2 Orange:LinearColor=<rgba 1 0.4167 0 10>
    R2 Pink:LinearColor=<rgba 1 0.1615 0.6951 7>
    R2 Purple:LinearColor=<rgba 3.1726 0.625 6 1>
    R2 Saffron:LinearColor=<rgba 1 0.7667 0 10>
    R2 SkyBlue:LinearColor=<rgba 0.1302 0.709 1 8>
    R2 TitaniumWhite:LinearColor=<rgba 10 10 10 1>
    R3 Black:LinearColor=<rgba 0.13 0.13 0.13 1.1>
    R3 BurntSienna:LinearColor=<rgba 0.35 0.0978 0.021 1>
    R3 Crimson:LinearColor=<rgba 1 0.03 0.0563 1>
    R3 DefaultColor:LinearColor=<rgba 0.8594 0.8641 1 0.8>
    R3 ForestGreen:LinearColor=<rgba 0.18 0.6 0.18 1>
    R3 Grey:LinearColor=<rgba 1 1 1 0.8>
    R3 Lime:LinearColor=<rgba 1 2 0 1>
    R3 Orange:LinearColor=<rgba 1 0.5917 0.02 1.2>
    R3 Pink:LinearColor=<rgba 1.7 0.51 1.2673 0.5>
    R3 Purple:LinearColor=<rgba 1 0.23 0.7433 1>
    R3 Saffron:LinearColor=<rgba 1 0.8133 0.2 1>
    R3 SkyBlue:LinearColor=<rgba 0.195 0.9303 1.3 1>
    R3 TitaniumWhite:LinearColor=<rgba 0.75 0.75 0.75 1.5>
    Saffron=<rgba 1 0.7667 0 0.01>
    SkyBlue=<rgba 0 0.357 0.5365 0.05>
    TitaniumWhite=<rgba 0.75 0.75 0.75 0.015>
    Use R1 Painted Overrides=true
    Use R2 Painted Overrides=true
    Use R3 Painted Overrides=true
   )
   &VC_InitializeRibbon_v1(
    Color=$StackContext.Color
    Custom Offset=true
    Enable Skeletal Position=true
    Lifetime=&MakeCustomFloatFromBool(Bool=$User.bIsFrontendPreview, True Float=0.45, False Float=0.45)
    Local Vector Offset=&AddVectorToPosition(
     Position=<50 0 0>
     Vector=&AddVectorToPosition(Position=$User.DTCarOffset, Vector=Linked() {linkedVariable {name=User.SingleBoostPos, type=NiagaraPosition}})
    )
    Offset Amount=1
    Ribbon Width:NiagaraFloat=&MakeCustomFloatFromBool(Bool=$User.bIsFrontendPreview, True Float=50, False Float=40)
   )
   &AddVelocity(
    Velocity=&MakeCustomVectorFromBool(Bool=$User.bIsFrontendPreview, True Vector=<4000 0 0>, False Vector=<1000 0 0>)
    Velocity Speed Scale=&MakeFloatFromVector(
     Vector=&VectorFromFloat(Value=&MakeCustomFloatFromBool(Bool=$User.bIsFrontendPreview, True Float=-1.5, False Float=-2))
     Channel=::NewEnumerator0
    )
   ) {module=AddVelocity001}
  ]
  particleUpdateScript [
   &ParticleState()
   &InheritSourceMovement(
    Applied Position Scale=&Multiply_VectorByFloat(
     Vector=<1 1 1>
     Float=&MakeCustomFloatFromBool(
      Bool=$User.bIsFrontendPreview
      True Float=0
      False Float=&FloatFromCurve(FloatCurve=DataInterface() {propertyValues {Curve {keys [
           {arriveTangent=-0.1077, interpMode=RCIM_Cubic, leaveTangent=-0.1077, tangentMode=RCTM_User, time=0.2, value=1}
           {arriveTangent=-0.0009, interpMode=RCIM_Cubic, leaveTangent=-0.0009, tangentMode=RCTM_User, time=1, value=0}
          ]}}})
     )
    )
    ApplyForce=false
    ApplyPosition=true
    ApplyPositionFalloff=false
   )
   &SolveForcesAndVelocity(Clamp Velocity=false, Limit Acceleration=false)
   &VC_UpdateRibbon_v1(Enable Skeletal Position=true)
   &ScaleColor(
    Scale Alpha=&FloatFromCurve(FloatCurve=DataInterface() {propertyValues {Curve {keys [<at 0 0>,<at 0.56 1.2>,<at 1 0>]}}}, Scale Curve=0.5)
    Scale RGB=&VectorFromCurve(VectorCurve=DataInterface() {propertyValues {
       XCurve {keys [<at 0 0.5>,<at 0.5 2.5>,<at 1 0.5>]}
       YCurve {keys [<at 0 0.5>,<at 0.5 2.5>,<at 1 0.5>]}
       ZCurve {keys [<at 0 0.5>,<at 0.5 2.5>,<at 1 0.5>]}
      }})
    ScaleA=true
    ScaleRGB=true
   ) {module=ScaleColor001}
  ]
  rendererClasses [&NiagaraRibbonRendererProperties]
  renderers [
   {rendererClass=&NiagaraRibbonRendererProperties}
   {rendererClass=&NiagaraRibbonRendererProperties, rendererIndex=1}
   {rendererClass=&NiagaraRibbonRendererProperties, rendererIndex=2}
   {rendererClass=&NiagaraRibbonRendererProperties, rendererIndex=3}
  ]
 }}
summary {
 systemName=NS_Boost_LightTrail
 emitters [{emitterName=NE_VC_Ribbon_Parent, rendererClasses [&NiagaraRibbonRendererProperties]}]
 userVariables [
  User.bIsFrontendPreview {type=NiagaraBool, defaultValue=true}
  User.bIsPainted {type=NiagaraBool}
  User.ColorVariantID {type=NiagaraInt32}
  User.DTCarOffset {type=NiagaraPosition, description="", defaultValue {}}
  User.GameModeID {type=NiagaraInt32}
  User.PaintedColor {type=LinearColor, description="", defaultValue {struct=LinearColor, value=<rgba 1 1 1 1>}}
  User.PaintedScalar {type=NiagaraFloat, defaultValue=1, description=""}
  User.PortCount {type=NiagaraInt32, defaultValue=2}
  User.SingleBoostPos {type=NiagaraPosition, description="", defaultValue {}}
 ]
}
